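; Permanent rift portal that teleports players, vehicles and AI pawns to a point at a tuned height.
; BeginPlay computes that destination and checks balance data to remove disabled rifts.
; Overlaps teleport on the server, except driven vehicles, which the driver's client moves; handling includes FX, fall-damage protection, staggered arrival offsets and quest credit outside Creative.
; A separate proximity capsule runs a timeline that widens the rift, with expand/collapse audio.
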(event EventBeginPlay
  ;[1 1 1] EVENT--> BeginPlay
  (|CallParentFunction)
  (Variables|RiftPortal|SetPortalOpen true)
  (Utilities|FlowControl|Delay 3.0)
  (Variables|SetTeleportLocation_X (Math|Vector|MakeVector (.x (Transformation|GetActorLocation)) (.y (Transformation|GetActorLocation)) (ScalableFloat|GetValueatLevel (Variables|RiftPortal|GetTeleportHeight))))
  (Utilities|FlowControl|Sequence
    (:then_0
      ;[1 0.05805 0.8703] What Mason changed
      (bind _outxy (DataTable|EvaluateCurveTableRow "/Game/Athena/Balance/DataTables/AthenaGameData.AthenaGameData" "Default.PermaRift.ShouldSpawn")
        (:RowFound
          (if (> _outxy 0.0)
            _
            (else
              _)))
        (:RowNotFound
          _))
      (Actor|DestroyActor)
      ;[1 1 1] Check if In Creative Mode or not for Quest
      (bind _asfort_game_state_athena (Utilities|Casting|CastToFortGameStateAthena (Game|GetGameState)))
      (if (Variables|/Script/FortniteGame.FortGameStateAthena|GetbPlaylistDataIsLoaded _asfort_game_state_athena)
        (bind _cancompletequest (Class|AthenaQuestFunctionLibrary|AthenaCheckCreativeMode (Game|GetGameState)))
        (if _cancompletequest
          (Variables|QuestStuff|SetCreativeMode false)
          (else
            (Variables|QuestStuff|SetCreativeMode true)))
        (else
          (Playlist|BindEventtoOnCurrentPlaylistLoaded _asfort_game_state_athena (delegate Custom|PlaylistLoaded)))))
    (:then_1
      ;[1 0.6592 0.5993] Failsafe check to see if Rift should spawn because having no delay doesn't seem to work consistently
      (Utilities|FlowControl|Delay 2e+01)
      (if (not (> (ScalableFloat|GetValueatLevel (Variables|RiftPortal|GetShouldSpawnRift?)) 0.0))
        (Actor|DestroyActor)))))

(event Custom|PlaylistLoaded
  (params (PlaylistName :type Name) (PlaylistContextTags :type Gameplay%20Tag%20Container%20Structure%20%28by%20ref%29))
  ;[1 1 1] Check if In Creative Mode or not for Quest
  (bind _cancompletequest (Class|AthenaQuestFunctionLibrary|AthenaCheckCreativeMode (Game|GetGameState)))
  (if _cancompletequest
    (Variables|QuestStuff|SetCreativeMode false)
    (else
      (Variables|QuestStuff|SetCreativeMode true))))

(event Custom|PlayCosmeticFX
  ;[1 1 1] FX for Big Sky crack can only play every 6 seconds
  (Utilities|FlowControl|Sequence
    (:then_0
      _)
    (:then_1
      (Utilities|FlowControl|Delay 6.0)))
  (Utilities|FlowControl|DoOnce false
    (:Completed
      (Utilities|FlowControl|Delay 1.0)
      (bind _returnvalue (Effects|Components|ParticleSystem|SpawnFXAtLocation (Variables|VFX|GetCrackExitRift) (+ (Variables|RiftPortal|GetTeleportLocation) "0, 0,200.000000") (Math|Rotator|MakeRotator 0.0 0.0 (Math|Random|RandomFloatinRange -1.8e+02 1.8e+02)) "1.000000,1.000000,1.000000" true "None" true))
      (LOD|SetMaxDrawDistance _returnvalue 0.0))))

(event Custom|PlayTeleportFX
  (params (PlayerLocation :type Vector))
  ;[1 0.6592 0.5993] Failsafe check to see if Rift should spawn because having no delay doesn't seem to work consistently
  (Effects|Components|ParticleSystem|SpawnFXAtLocation (Variables|VFX|GetEnterRift_NS) PlayerLocation)
  ;[1 1 1] Spawn teleport visuals and audio for clients
  (Audio|PlaySoundatLocation (Variables|Sound|GetEnteredRift) PlayerLocation)
  (Utilities|FlowControl|Delay 1.0)
  (Audio|SpawnSoundatLocation (Variables|Sound|GetPlayerExitSkyCrack) (Variables|RiftPortal|GetTeleportLocation)))

(event OnComponentBeginOverlap(OverlapCapsule)
  (params (OverlappedComponent :type Primitive%20Component%20Object%20Reference) (OtherActor :type Actor%20Object%20Reference) (OtherComp :type Primitive%20Component%20Object%20Reference) (OtherBodyIndex :type Integer) (bFromSweep :type Boolean) (SweepResult :type Hit%20Result%20Structure%20%28by%20ref%29))
  (bind _vehicleinterface (Variables|RiftPortal|GetVehicleInterface))
  (bind _returnvalue (Vehicle|GetDriver _vehicleinterface))
  (bind (_actorrotation_roll _actorrotation_pitch _actorrotation_yaw) (Math|Rotator|BreakRotator (Variables|GetActorRotation_Roll)))
  (bind _playerpawn (Variables|RiftPortal|GetPlayerPawn))
  (bind _overlapactor (Variables|RiftPortal|GetOverlapActor))
  ;[1 1 1] Teleport Player and/or Vehicle
  (Variables|RiftPortal|SetOverlapActor OtherActor)
  ;;[1 0.526 0.526] Failsafe to ensure Rift is supposed to spawn
  (if (> (ScalableFloat|GetValueatLevel (Variables|RiftPortal|GetShouldSpawnRift?)) 0.0)
    (if (Variables|RiftPortal|GetPortalOpen)
      (Utilities|FlowControl|Sequence
        (:then_0
          (Variables|RiftPortal|SetActorRotation (Transformation|GetActorRotation (Variables|RiftPortal|GetOverlapActor)))
          (if (Networking|HasAuthority)
            (Utilities|IsValid _overlapactor
              (:Is%20Valid
                (bind _asfort_player_pawn (Utilities|Casting|CastToFortPlayerPawn _overlapactor)
                  (:then
                    (if (not (Variables|/Script/FortniteGame.FortPawn|GetbIsDBNO _asfort_player_pawn))
                      (Variables|RiftPortal|SetPlayerPawn _asfort_player_pawn)
                      (CallFunction|EjectfromTurret)
                      (CallFunction|PlayCosmeticFX)
                      (CallFunction|PlayTeleportFX (Transformation|GetActorLocation _asfort_player_pawn))
                      (if (Vehicle|IsinVehicle _asfort_player_pawn)
                        (if (not (Vehicle|HasDriver (Vehicle|GetVehicleInterface _asfort_player_pawn)))
                          (Variables|RiftPortal|SetVehicleInterface (Vehicle|GetVehicleInterface _asfort_player_pawn))
                          (CallFunction|VehicleTeleportEvent))
                        (else
                          (Networking|SwitchHasAuthority
                            (:Authority
                              (RemoteControlPawn|ForceDetachfromRemoteControlledPawn _asfort_player_pawn)
                              (Variables|/Script/FortniteGame.FortPawn|SetbIgnoreNextFallingDamage true _asfort_player_pawn) ; Have to do this because of the delay. Otherwise player can die from fall damage if they fall from high distance.
                              (bind _newparam (CallFunction|IsFlying))
                              (if (not _newparam)
                                (Utilities|FlowControl|Sequence
                                  (:then_0
                                    (Parachute|EndSkydiving _asfort_player_pawn) ; End Skydiving if player was previously
                                    (Transform|Actor|TeleportPlayerPawn _asfort_player_pawn (Variables|RiftPortal|GetTeleportLocation) (Math|Rotator|MakeRotator 0.0 0.0 _actorrotation_yaw) true false)
                                    (Ability|SendGameplayEventtoActor _asfort_player_pawn "(TagName=\"Abilities.Generic.Rift.Trigger\")"))
                                  (:then_1
                                    (Variables|RiftPortal|SetTeleportLocation (+ (Variables|RiftPortal|GetTeleportLocation) "0,100.000000, 0")) ; Offset players so they don't all teleport on top of each other
                                    ;[1 1 1] SERVER: Check to Award Single Use Quest Progression if not in Creative Mode
                                    (if (not (Variables|QuestStuff|GetCreativeMode))
                                      (Networking|SwitchHasAuthority
                                        (:Authority
                                          (Utilities|IsValid _playerpawn
                                            (:Is%20Valid
                                              (CallFunction|UseRiftQuestUpdate (Pawn|GetController _playerpawn)))
                                            (:Is%20Not%20Valid
                                              (Utilities|IsValid (Utilities|ToObject(Interface) _vehicleinterface)
                                                (:Is%20Valid
                                                  (Utilities|IsValid _returnvalue
                                                    (:Is%20Valid
                                                      (CallFunction|UseRiftQuestUpdate (Pawn|GetController _returnvalue)))
                                                    (:Is%20Not%20Valid
                                                      _))
                                                  (for _array_element (Vehicle|GetAllPassengers _vehicleinterface)
                                                    (Utilities|IsValid _array_element
                                                      (:Is%20Valid
                                                        (CallFunction|UseRiftQuestUpdate (Pawn|GetController _array_element)))))))))))))))))))))
                  (:CastFailed
                    (bind _asfort_athena_vehicle (Utilities|Casting|CastToFortAthenaVehicle _overlapactor)
                      (:then
                        (bind _output_get (Variables|RiftPortal|SetVehicleInterface _asfort_athena_vehicle))
                        (if (Vehicle|HasDriver _output_get)
                          _
                          (elif (== (Utilities|Array|Length (Vehicle|GetAllPassengers _output_get)) 0)
                            (bind _newparam_2 (CallFunction|IsFlying))
                            (if (not _newparam_2)
                              (CallFunction|VehicleTeleportEvent)))))
                      (:CastFailed
                        (bind _asfort_aipawn (Utilities|Casting|CastToFortAIPawn _overlapactor))
                        (Utilities|IsValid (.AttachedPlayerPawn (Default|IsPlayerAttachedtoAIPawn _asfort_aipawn))
                          (:Is%20Valid
                            (CallFunction|TryRiftTeleport (.AttachedPlayerPawn (Default|IsPlayerAttachedtoAIPawn _asfort_aipawn))))
                          (:Is%20Not%20Valid
                            _))
                        (CallFunction|PlayTeleportFX (Transformation|GetActorLocation _asfort_aipawn))
                        (AI|TeleportAIPawn _asfort_aipawn (Math|Vector|vector+vector (Variables|RiftPortal|GetTeleportLocation) (1e+02 1e+02 0.0)) "0, 0, 0" true)
                        (Variables|/Script/FortniteGame.FortPawn|SetbIgnoreNextFallingDamage true _asfort_aipawn)
                        (Ability|SendGameplayEventtoActor _asfort_aipawn "(TagName=\"Abilities.Generic.Rift.Trigger\")")))))))
            (else
              (bind _asfort_player_pawn_2 (Utilities|Casting|CastToFortPlayerPawn (Variables|RiftPortal|GetOverlapActor)))
              (Variables|RiftPortal|SetPlayerPawn _asfort_player_pawn_2)
              (if (and (Vehicle|IsDrivingVehicle (Variables|RiftPortal|GetPlayerPawn)) (Pawn|IsLocallyControlled (Variables|RiftPortal|GetPlayerPawn)))
                (CallFunction|EjectfromTurret)
                (bind _output_get_2 (Variables|RiftPortal|SetVehicleInterface (Vehicle|GetVehicleInterface (Variables|RiftPortal|GetPlayerPawn))))
                (if (Vehicle|CanTeleport (Variables|RiftPortal|GetVehicleInterface))
                  (bind _returnvalue_2 (Vehicle|GetVehicleActor _output_get_2 :!class /Script/FortniteGame.FortVehicleInterface))
                  (CallFunction|TeleportVehicle _returnvalue_2 (Variables|RiftPortal|GetTeleportLocation))
                  (Vehicle|SetIgnoreNextFallingDamage (Variables|RiftPortal|GetVehicleInterface) true))))))
        (:then_1
          ;;[1 1 1] Detach if swinging -- fixes FORT-421121
          (bind _asplayer_pawn_athena (Utilities|Casting|CastToPlayerPawn_Athena (Variables|RiftPortal|GetOverlapActor)))
          (if (Swinging|GetIsSwinging _asplayer_pawn_athena)
            (Utilities|IsValid (Pawn|GetController _asplayer_pawn_athena)
              (:Is%20Valid
                (Swinging|EndSwingingMovement (Actor|GetComponentbyClass (Pawn|GetController _asplayer_pawn_athena) "/Script/FortniteGame.FortControllerComponent_Swinging"))))))))
    (else
      (Actor|DestroyActor))))

(event Custom|VehicleTeleportEvent
  ;[1 1 1] Teleport Player and/or Vehicle
  (if (Vehicle|CanTeleport (Variables|RiftPortal|GetVehicleInterface))
    (bind _returnvalue (Vehicle|GetVehicleActor (stale OnComponentBeginOverlap(OverlapCapsule) _output_get_2) :!class /Script/FortniteGame.FortVehicleInterface))
    (CallFunction|TeleportVehicle _returnvalue (Variables|RiftPortal|GetTeleportLocation))
    (Vehicle|SetIgnoreNextFallingDamage (Variables|RiftPortal|GetVehicleInterface) true)))

(event EventOverlappingOnBeginPlay
  (params (Actor :type Actor%20Object%20Reference))
  (bind _vehicleinterface (Variables|RiftPortal|GetVehicleInterface))
  (bind _returnvalue (Vehicle|GetDriver _vehicleinterface))
  (bind (_actorrotation_roll _actorrotation_pitch _actorrotation_yaw) (Math|Rotator|BreakRotator (Variables|GetActorRotation_Roll)))
  (bind _playerpawn (Variables|RiftPortal|GetPlayerPawn))
  (bind _overlapactor (Variables|RiftPortal|GetOverlapActor))
  ;[1 1 1] Teleport Player and/or Vehicle
  (Variables|RiftPortal|SetOverlapActor Actor)
  ;;[1 0.526 0.526] Failsafe to ensure Rift is supposed to spawn
  (if (> (ScalableFloat|GetValueatLevel (Variables|RiftPortal|GetShouldSpawnRift?)) 0.0)
    (if (Variables|RiftPortal|GetPortalOpen)
      (Utilities|FlowControl|Sequence
        (:then_0
          (Variables|RiftPortal|SetActorRotation (Transformation|GetActorRotation (Variables|RiftPortal|GetOverlapActor)))
          (if (Networking|HasAuthority)
            (Utilities|IsValid _overlapactor
              (:Is%20Valid
                (bind _asfort_player_pawn (Utilities|Casting|CastToFortPlayerPawn _overlapactor)
                  (:then
                    (if (not (Variables|/Script/FortniteGame.FortPawn|GetbIsDBNO _asfort_player_pawn))
                      (Variables|RiftPortal|SetPlayerPawn _asfort_player_pawn)
                      (CallFunction|EjectfromTurret)
                      (CallFunction|PlayCosmeticFX)
                      (CallFunction|PlayTeleportFX (Transformation|GetActorLocation _asfort_player_pawn))
                      (if (Vehicle|IsinVehicle _asfort_player_pawn)
                        (if (not (Vehicle|HasDriver (Vehicle|GetVehicleInterface _asfort_player_pawn)))
                          (Variables|RiftPortal|SetVehicleInterface (Vehicle|GetVehicleInterface _asfort_player_pawn))
                          (CallFunction|VehicleTeleportEvent))
                        (else
                          (Networking|SwitchHasAuthority
                            (:Authority
                              (RemoteControlPawn|ForceDetachfromRemoteControlledPawn _asfort_player_pawn)
                              (Variables|/Script/FortniteGame.FortPawn|SetbIgnoreNextFallingDamage true _asfort_player_pawn) ; Have to do this because of the delay. Otherwise player can die from fall damage if they fall from high distance.
                              (bind _newparam (CallFunction|IsFlying))
                              (if (not _newparam)
                                (Utilities|FlowControl|Sequence
                                  (:then_0
                                    (Parachute|EndSkydiving _asfort_player_pawn) ; End Skydiving if player was previously
                                    (Transform|Actor|TeleportPlayerPawn _asfort_player_pawn (Variables|RiftPortal|GetTeleportLocation) (Math|Rotator|MakeRotator 0.0 0.0 _actorrotation_yaw) true false)
                                    (Ability|SendGameplayEventtoActor _asfort_player_pawn "(TagName=\"Abilities.Generic.Rift.Trigger\")"))
                                  (:then_1
                                    (Variables|RiftPortal|SetTeleportLocation (+ (Variables|RiftPortal|GetTeleportLocation) "0,100.000000, 0")) ; Offset players so they don't all teleport on top of each other
                                    ;[1 1 1] SERVER: Check to Award Single Use Quest Progression if not in Creative Mode
                                    (if (not (Variables|QuestStuff|GetCreativeMode))
                                      (Networking|SwitchHasAuthority
                                        (:Authority
                                          (Utilities|IsValid _playerpawn
                                            (:Is%20Valid
                                              (CallFunction|UseRiftQuestUpdate (Pawn|GetController _playerpawn)))
                                            (:Is%20Not%20Valid
                                              (Utilities|IsValid (Utilities|ToObject(Interface) _vehicleinterface)
                                                (:Is%20Valid
                                                  (Utilities|IsValid _returnvalue
                                                    (:Is%20Valid
                                                      (CallFunction|UseRiftQuestUpdate (Pawn|GetController _returnvalue)))
                                                    (:Is%20Not%20Valid
                                                      _))
                                                  (for _array_element (Vehicle|GetAllPassengers _vehicleinterface)
                                                    (Utilities|IsValid _array_element
                                                      (:Is%20Valid
                                                        (CallFunction|UseRiftQuestUpdate (Pawn|GetController _array_element)))))))))))))))))))))
                  (:CastFailed
                    (bind _asfort_athena_vehicle (Utilities|Casting|CastToFortAthenaVehicle _overlapactor)
                      (:then
                        (bind _output_get (Variables|RiftPortal|SetVehicleInterface _asfort_athena_vehicle))
                        (if (Vehicle|HasDriver _output_get)
                          _
                          (elif (== (Utilities|Array|Length (Vehicle|GetAllPassengers _output_get)) 0)
                            (bind _newparam_2 (CallFunction|IsFlying))
                            (if (not _newparam_2)
                              (CallFunction|VehicleTeleportEvent)))))
                      (:CastFailed
                        (bind _asfort_aipawn (Utilities|Casting|CastToFortAIPawn _overlapactor))
                        (Utilities|IsValid (.AttachedPlayerPawn (Default|IsPlayerAttachedtoAIPawn _asfort_aipawn))
                          (:Is%20Valid
                            (CallFunction|TryRiftTeleport (.AttachedPlayerPawn (Default|IsPlayerAttachedtoAIPawn _asfort_aipawn))))
                          (:Is%20Not%20Valid
                            _))
                        (CallFunction|PlayTeleportFX (Transformation|GetActorLocation _asfort_aipawn))
                        (AI|TeleportAIPawn _asfort_aipawn (Math|Vector|vector+vector (Variables|RiftPortal|GetTeleportLocation) (1e+02 1e+02 0.0)) "0, 0, 0" true)
                        (Variables|/Script/FortniteGame.FortPawn|SetbIgnoreNextFallingDamage true _asfort_aipawn)
                        (Ability|SendGameplayEventtoActor _asfort_aipawn "(TagName=\"Abilities.Generic.Rift.Trigger\")")))))))
            (else
              (bind _asfort_player_pawn_2 (Utilities|Casting|CastToFortPlayerPawn (Variables|RiftPortal|GetOverlapActor)))
              (Variables|RiftPortal|SetPlayerPawn _asfort_player_pawn_2)
              (if (and (Vehicle|IsDrivingVehicle (Variables|RiftPortal|GetPlayerPawn)) (Pawn|IsLocallyControlled (Variables|RiftPortal|GetPlayerPawn)))
                (CallFunction|EjectfromTurret)
                (bind _output_get_2 (Variables|RiftPortal|SetVehicleInterface (Vehicle|GetVehicleInterface (Variables|RiftPortal|GetPlayerPawn))))
                (if (Vehicle|CanTeleport (Variables|RiftPortal|GetVehicleInterface))
                  (bind _returnvalue_2 (Vehicle|GetVehicleActor _output_get_2 :!class /Script/FortniteGame.FortVehicleInterface))
                  (CallFunction|TeleportVehicle _returnvalue_2 (Variables|RiftPortal|GetTeleportLocation))
                  (Vehicle|SetIgnoreNextFallingDamage (Variables|RiftPortal|GetVehicleInterface) true))))))
        (:then_1
          ;;[1 1 1] Detach if swinging -- fixes FORT-421121
          (bind _asplayer_pawn_athena (Utilities|Casting|CastToPlayerPawn_Athena (Variables|RiftPortal|GetOverlapActor)))
          (if (Swinging|GetIsSwinging _asplayer_pawn_athena)
            (Utilities|IsValid (Pawn|GetController _asplayer_pawn_athena)
              (:Is%20Valid
                (Swinging|EndSwingingMovement (Actor|GetComponentbyClass (Pawn|GetController _asplayer_pawn_athena) "/Script/FortniteGame.FortControllerComponent_Swinging"))))))))
    (else
      (Actor|DestroyActor))))

(event Custom|TryRiftTeleport
  (params (TargetActor :type Actor%20Object%20Reference))
  (bind _vehicleinterface (Variables|RiftPortal|GetVehicleInterface))
  (bind _returnvalue (Vehicle|GetDriver _vehicleinterface))
  (bind (_actorrotation_roll _actorrotation_pitch _actorrotation_yaw) (Math|Rotator|BreakRotator (Variables|GetActorRotation_Roll)))
  (bind _playerpawn (Variables|RiftPortal|GetPlayerPawn))
  (bind _overlapactor (Variables|RiftPortal|GetOverlapActor))
  ;[1 1 1] Teleport Player and/or Vehicle
  (Variables|RiftPortal|SetOverlapActor TargetActor)
  ;;[1 0.526 0.526] Failsafe to ensure Rift is supposed to spawn
  (if (> (ScalableFloat|GetValueatLevel (Variables|RiftPortal|GetShouldSpawnRift?)) 0.0)
    (if (Variables|RiftPortal|GetPortalOpen)
      (Utilities|FlowControl|Sequence
        (:then_0
          (Variables|RiftPortal|SetActorRotation (Transformation|GetActorRotation (Variables|RiftPortal|GetOverlapActor)))
          (if (Networking|HasAuthority)
            (Utilities|IsValid _overlapactor
              (:Is%20Valid
                (bind _asfort_player_pawn (Utilities|Casting|CastToFortPlayerPawn _overlapactor)
                  (:then
                    (if (not (Variables|/Script/FortniteGame.FortPawn|GetbIsDBNO _asfort_player_pawn))
                      (Variables|RiftPortal|SetPlayerPawn _asfort_player_pawn)
                      (CallFunction|EjectfromTurret)
                      (CallFunction|PlayCosmeticFX)
                      (CallFunction|PlayTeleportFX (Transformation|GetActorLocation _asfort_player_pawn))
                      (if (Vehicle|IsinVehicle _asfort_player_pawn)
                        (if (not (Vehicle|HasDriver (Vehicle|GetVehicleInterface _asfort_player_pawn)))
                          (Variables|RiftPortal|SetVehicleInterface (Vehicle|GetVehicleInterface _asfort_player_pawn))
                          (CallFunction|VehicleTeleportEvent))
                        (else
                          (Networking|SwitchHasAuthority
                            (:Authority
                              (RemoteControlPawn|ForceDetachfromRemoteControlledPawn _asfort_player_pawn)
                              (Variables|/Script/FortniteGame.FortPawn|SetbIgnoreNextFallingDamage true _asfort_player_pawn) ; Have to do this because of the delay. Otherwise player can die from fall damage if they fall from high distance.
                              (bind _newparam (CallFunction|IsFlying))
                              (if (not _newparam)
                                (Utilities|FlowControl|Sequence
                                  (:then_0
                                    (Parachute|EndSkydiving _asfort_player_pawn) ; End Skydiving if player was previously
                                    (Transform|Actor|TeleportPlayerPawn _asfort_player_pawn (Variables|RiftPortal|GetTeleportLocation) (Math|Rotator|MakeRotator 0.0 0.0 _actorrotation_yaw) true false)
                                    (Ability|SendGameplayEventtoActor _asfort_player_pawn "(TagName=\"Abilities.Generic.Rift.Trigger\")"))
                                  (:then_1
                                    (Variables|RiftPortal|SetTeleportLocation (+ (Variables|RiftPortal|GetTeleportLocation) "0,100.000000, 0")) ; Offset players so they don't all teleport on top of each other
                                    ;[1 1 1] SERVER: Check to Award Single Use Quest Progression if not in Creative Mode
                                    (if (not (Variables|QuestStuff|GetCreativeMode))
                                      (Networking|SwitchHasAuthority
                                        (:Authority
                                          (Utilities|IsValid _playerpawn
                                            (:Is%20Valid
                                              (CallFunction|UseRiftQuestUpdate (Pawn|GetController _playerpawn)))
                                            (:Is%20Not%20Valid
                                              (Utilities|IsValid (Utilities|ToObject(Interface) _vehicleinterface)
                                                (:Is%20Valid
                                                  (Utilities|IsValid _returnvalue
                                                    (:Is%20Valid
                                                      (CallFunction|UseRiftQuestUpdate (Pawn|GetController _returnvalue)))
                                                    (:Is%20Not%20Valid
                                                      _))
                                                  (for _array_element (Vehicle|GetAllPassengers _vehicleinterface)
                                                    (Utilities|IsValid _array_element
                                                      (:Is%20Valid
                                                        (CallFunction|UseRiftQuestUpdate (Pawn|GetController _array_element)))))))))))))))))))))
                  (:CastFailed
                    (bind _asfort_athena_vehicle (Utilities|Casting|CastToFortAthenaVehicle _overlapactor)
                      (:then
                        (bind _output_get (Variables|RiftPortal|SetVehicleInterface _asfort_athena_vehicle))
                        (if (Vehicle|HasDriver _output_get)
                          _
                          (elif (== (Utilities|Array|Length (Vehicle|GetAllPassengers _output_get)) 0)
                            (bind _newparam_2 (CallFunction|IsFlying))
                            (if (not _newparam_2)
                              (CallFunction|VehicleTeleportEvent)))))
                      (:CastFailed
                        (bind _asfort_aipawn (Utilities|Casting|CastToFortAIPawn _overlapactor))
                        (Utilities|IsValid (.AttachedPlayerPawn (Default|IsPlayerAttachedtoAIPawn _asfort_aipawn))
                          (:Is%20Valid
                            (CallFunction|TryRiftTeleport (.AttachedPlayerPawn (Default|IsPlayerAttachedtoAIPawn _asfort_aipawn))))
                          (:Is%20Not%20Valid
                            _))
                        (CallFunction|PlayTeleportFX (Transformation|GetActorLocation _asfort_aipawn))
                        (AI|TeleportAIPawn _asfort_aipawn (Math|Vector|vector+vector (Variables|RiftPortal|GetTeleportLocation) (1e+02 1e+02 0.0)) "0, 0, 0" true)
                        (Variables|/Script/FortniteGame.FortPawn|SetbIgnoreNextFallingDamage true _asfort_aipawn)
                        (Ability|SendGameplayEventtoActor _asfort_aipawn "(TagName=\"Abilities.Generic.Rift.Trigger\")")))))))
            (else
              (bind _asfort_player_pawn_2 (Utilities|Casting|CastToFortPlayerPawn (Variables|RiftPortal|GetOverlapActor)))
              (Variables|RiftPortal|SetPlayerPawn _asfort_player_pawn_2)
              (if (and (Vehicle|IsDrivingVehicle (Variables|RiftPortal|GetPlayerPawn)) (Pawn|IsLocallyControlled (Variables|RiftPortal|GetPlayerPawn)))
                (CallFunction|EjectfromTurret)
                (bind _output_get_2 (Variables|RiftPortal|SetVehicleInterface (Vehicle|GetVehicleInterface (Variables|RiftPortal|GetPlayerPawn))))
                (if (Vehicle|CanTeleport (Variables|RiftPortal|GetVehicleInterface))
                  (bind _returnvalue_2 (Vehicle|GetVehicleActor _output_get_2 :!class /Script/FortniteGame.FortVehicleInterface))
                  (CallFunction|TeleportVehicle _returnvalue_2 (Variables|RiftPortal|GetTeleportLocation))
                  (Vehicle|SetIgnoreNextFallingDamage (Variables|RiftPortal|GetVehicleInterface) true))))))
        (:then_1
          ;;[1 1 1] Detach if swinging -- fixes FORT-421121
          (bind _asplayer_pawn_athena (Utilities|Casting|CastToPlayerPawn_Athena (Variables|RiftPortal|GetOverlapActor)))
          (if (Swinging|GetIsSwinging _asplayer_pawn_athena)
            (Utilities|IsValid (Pawn|GetController _asplayer_pawn_athena)
              (:Is%20Valid
                (Swinging|EndSwingingMovement (Actor|GetComponentbyClass (Pawn|GetController _asplayer_pawn_athena) "/Script/FortniteGame.FortControllerComponent_Swinging"))))))))
    (else
      (Actor|DestroyActor))))

(event EventNotifyTeleportedVehicle
  (params (VehicleActor :type Actor%20Object%20Reference))
  (bind _vehicleinterface (Variables|RiftPortal|GetVehicleInterface))
  (bind _returnvalue (Vehicle|GetDriver _vehicleinterface))
  (bind _playerpawn (Variables|RiftPortal|GetPlayerPawn))
  ;[1 1 1] Teleport Player and/or Vehicle
  (bind _asfort_vehicle_interface (Utilities|Casting|CastToFortVehicleInterface VehicleActor))
  (bind _output_get (Variables|RiftPortal|SetVehicleInterface _asfort_vehicle_interface))
  (bind _returnvalue_2 (Vehicle|GetVehicleActor _output_get :!class /Script/FortniteGame.FortVehicleInterface))
  (CallFunction|PlayTeleportFX (Transformation|GetActorLocation _returnvalue_2))
  (bind _output_get_2 (Variables|RiftPortal|SetPassengerArray (Vehicle|GetAllPassengers _output_get)))
  (for _array_element _output_get_2
    (RemoteControlPawn|ForceDetachfromRemoteControlledPawn _array_element)
    (Ability|SendGameplayEventtoActor _array_element "(TagName=\"Abilities.Generic.Rift.Trigger\")")
    (Variables|RiftPortal|SetTeleportLocation (+ (Variables|RiftPortal|GetTeleportLocation) "0,100.000000, 0")) ; Offset players so they don't all teleport on top of each other
    ;[1 1 1] SERVER: Check to Award Single Use Quest Progression if not in Creative Mode
    (if (not (Variables|QuestStuff|GetCreativeMode))
      (Networking|SwitchHasAuthority
        (:Authority
          (Utilities|IsValid _playerpawn
            (:Is%20Valid
              (CallFunction|UseRiftQuestUpdate (Pawn|GetController _playerpawn)))
            (:Is%20Not%20Valid
              (Utilities|IsValid (Utilities|ToObject(Interface) _vehicleinterface)
                (:Is%20Valid
                  (Utilities|IsValid _returnvalue
                    (:Is%20Valid
                      (CallFunction|UseRiftQuestUpdate (Pawn|GetController _returnvalue)))
                    (:Is%20Not%20Valid
                      _))
                  (for _array_element_2 (Vehicle|GetAllPassengers _vehicleinterface)
                    (Utilities|IsValid _array_element_2
                      (:Is%20Valid
                        (CallFunction|UseRiftQuestUpdate (Pawn|GetController _array_element_2))))))))))))))

(event OnComponentBeginOverlap(Capsule)
  (params (OverlappedComponent :type Primitive%20Component%20Object%20Reference) (OtherActor :type Actor%20Object%20Reference) (OtherComp :type Primitive%20Component%20Object%20Reference) (OtherBodyIndex :type Integer) (bFromSweep :type Boolean) (SweepResult :type Hit%20Result%20Structure%20%28by%20ref%29))
  (bind _sm_rift (Variables|Default|GetSM_Rift))
  ;[1 1 1] On Overlap - Toggle Post Process
  (Utilities|Casting|CastToFortPlayerPawn OtherActor)
  (CallFunction|PlayExpandorCollapseAudio true)
  ;;[1 1 1] Widen rift if player is close
  (bind (_direction _open) (Timeline|VectorScale Play)
    (:Update
      (Rendering|Material|SetScalarParameterValueonMaterials _sm_rift "Open" (Math|Float|SelectFloat 0.0 _open))
      (Transformation|SetRelativeRotation _sm_rift (Math|Rotator|MakeRotator (* _open -2e+01)) false false)
      (Effects|Components|ParticleSystem|SetFloatParameter (Variables|Default|GetNS_Rift_Idle_Loop) "Open" _open))))

(event OnComponentEndOverlap(Capsule)
  (params (OverlappedComponent :type Primitive%20Component%20Object%20Reference) (OtherActor :type Actor%20Object%20Reference) (OtherComp :type Primitive%20Component%20Object%20Reference) (OtherBodyIndex :type Integer))
  (bind _sm_rift (Variables|Default|GetSM_Rift))
  ;[1 1 1] On Overlap - Toggle Post Process
  (if (<= (Utilities|Array|Length (Collision|GetOverlappingActors OtherActor "/Script/FortniteGame.FortPlayerPawn")) 0)
    (CallFunction|PlayExpandorCollapseAudio false)
    ;;[1 1 1] Widen rift if player is close
    (Timeline|VectorScale Reverse)))

(fn CheckForRiftLocUpdate
  (params (InController :type Controller%20Object%20Reference)))

(fn EjectFromTurret
  (bind _playerpawn (Variables|RiftPortal|GetPlayerPawn))
  (if (Vehicle|IsinVehicle _playerpawn)
    (if (not (Vehicle|CanTeleport (Vehicle|GetVehicleInterface _playerpawn)))
      (GameplayEffects|RemoveActiveEffectswithGrantedTags (Ability|GetAbilitySystemComponent _playerpawn) "(GameplayTags=((TagName=\"Gameplay.Action.Player.AthenaInVehicle\")))") ; If vehicle isn't supposed to teleport then kick the player from it
    ))
  (return))

(fn IsFlying
  (bind _playerpawn (Variables|RiftPortal|GetPlayerPawn))
  (Utilities|IsValid _playerpawn
    (:Is%20Valid
      (bind _asfort_movement_comp_character (Utilities|Casting|CastToFortMovementComp_Character (Pawn|GetMovementComponent _playerpawn))
        (:then
          (return (Utilities|Enum|Equal(Enum) (Variables|/Script/Engine.CharacterMovementComponent|GetMovementMode _asfort_movement_comp_character) "MOVE_Flying")))
        (:CastFailed
          (return false))))))

(fn IsPlayerAttachedToAIPawn
  (params (AIPawn :type Fort%20AIPawn%20Object%20Reference))
  (for _array_element (Actor|GetAttachedActors AIPawn true false)
    (bind _asfort_player_pawn (Utilities|Casting|CastToFortPlayerPawn _array_element))
    (return true _asfort_player_pawn))
  (return false 0))

(fn OnRep_PlayerPawn)

(fn OnRep_TeleportLocation)

(fn OnRep_Vehicle)

(fn PlayExpandOrCollapseAudio
  (params (Expand :type Boolean))
  (if (Variables|RiftPortal|GetPortalOpen)
    (Audio|PlaySoundatLocation (select Expand "/Game/Sounds/Fort_GamePlay_Sounds/Player/RiftPortal/RiftPortal_Expand_Cue.RiftPortal_Expand_Cue" "/Game/Sounds/Fort_GamePlay_Sounds/Player/RiftPortal/RiftPortal_Collapse_Cue.RiftPortal_Collapse_Cue") (Transformation|GetWorldLocation (Variables|Default|GetSM_Rift)))))

(fn ConstructionScript)

(fn UseRiftQuestUpdate
  (params (InController :type Controller%20Object%20Reference))
  (Utilities|IsValid InController
    (:Is%20Valid
      (bind _asfort_player_controller (Utilities|Casting|CastToFortPlayerController InController))
      (bind (_questisvalid _questcompleted) (Default|AthenaCheckQuestandObjectiveCompletion _asfort_player_controller (Variables|QuestStuff|GetQuestItem) (Variables|QuestStuff|GetObjectiveBackendName)))
      (if (and _questisvalid (not _questcompleted))
        (Class|AthenaQuestFunctionLibrary|AthenaGiveQuestUpdatetoPlayers :InQuestItemRef (Variables|QuestStuff|GetQuestItem) :ObjectiveBackendName (Variables|QuestStuff|GetObjectiveBackendName) :PlayerControllersForUpdate (Utilities|Array|MakeArray _asfort_player_controller))))))
